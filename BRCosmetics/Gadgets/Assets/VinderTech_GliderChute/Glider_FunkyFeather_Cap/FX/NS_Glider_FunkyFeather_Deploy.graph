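# Niagara deploy effect for the FunkyFeather glider that bursts music-note meshes and sparkle sprites after a short delay.
# Notes_Out sprays outward, Notes_Up rises from a ring and falls, and Notes_Wings fires staggered bursts in boxes at glider sockets chosen by spawn group and partition.
# An integer variant drives colors, note renderer visibility and sparkle spawning.
# Front-end modifiers adjust timing and offsets, while particles partly follow source movement.
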NiagaraSystem /BRCosmetics/Gadgets/Assets/VinderTech_GliderChute/Glider_FunkyFeather_Cap/FX/NS_Glider_FunkyFeather_Deploy
dependencies {
 usedDataInterfaces [
  &NiagaraDataInterfaceCurve
  &NiagaraDataInterfaceArrayColor
  &NiagaraDataInterfaceMeshRendererInfo
  &NiagaraDataInterfaceVectorCurve
  &NiagaraDataInterfaceArrayInt32
  &NiagaraDataInterfaceArrayFloat3
  &NiagaraDataInterfaceSocketReader
 ]
 usedDynamicInputs [
  &Multiply_Vector2D_ByFloat
  &FNTimeOfDayColor
  &MakeIntFromFloat
 ]
 usedModules [
  &SystemState
  &self:SystemSpawnScript.SystemScriptSource.SystemScriptGraph.NiagaraNodeAssignment_0.SetVariables_#1
  &self:SystemSpawnScript.SystemScriptSource.SystemScriptGraph.NiagaraNodeAssignment_3.SetVariables_#2
  &self:SystemSpawnScript.SystemScriptSource.SystemScriptGraph.NiagaraNodeAssignment_1.SetVariables_#3
 ]
}
emitters {
 Notes_Out {
  emitterUpdateScript [
   &EmitterState(
    Inactive Response=::NewEnumerator1
    Life Cycle Mode=::NewEnumerator1
    Loop Behavior=::NewEnumerator1
    Loop Delay=&FN_FrontEndFloatModifier(Gameplay=0.9, Front End=0.65)
    Loop Duration=2.5
    UseLoopDelay=true
   )
   &SpawnBurst_Instantaneous(
    Spawn Count=&RandomRangeInt(Minimum=14, Maximum=16)
    Use Spawn Probability=false
    UseLoopCountLimit=false
   )
  ]
  particleSpawnScript [
   /SetVariables(Particles.StrengthRandom:NiagaraFloat=&RandomRangeFloat(Minimum=0.75, Maximum=1)) {isSetParameters=true, module=SetVariables_#4}
   &InitializeParticle(
    Color=&MaskLinearColor(
     Mask Source=1
     Mask Parameter=$System.intVariant
     Mask=0
     Masked Linear Color=&SelectLinearColorFromArray(
      Array Sampling Mode=::NewEnumerator0
      Direct Array Index=?
      Color Selection Array=DataInterface() {propertyValues {ColorData [<rgba 0.36 0.9 0.5535 1>, <rgba 0.8019 0.5 1 1>, <rgba 1 0.5127 0.3 1>]}}
      Random Seed=0
     )
     Passthrough Linear Color=&SelectLinearColorFromArray(
      Array Sampling Mode=::NewEnumerator1
      Direct Array Index=$System.intVariant
      Color Selection Array=DataInterface() {propertyValues {ColorData [<rgba 1 0.5127 0.3 1>, <rgba 1 0.7061 0.3217 1>, <rgba 0.2321 0.2437 0.2696 1>]}}
      Random Seed=?
     )
    )
    Lifetime=&RandomRangeFloat(Minimum=0.6, Maximum=0.9)
    Mesh Renderer Array Visibility Mode=::NewEnumerator3
    Mesh Renderer Info=DataInterface() {propertyValues {MeshRenderer=&self:Notes_Out_1.NiagaraMeshRendererProperties_1}}
    Mesh Scale Mode=::NewEnumerator3
    Position Offset=&FN_FrontEndVector3Modifier(Gameplay=<0 0 0>, Front End=<0 -40 80>)
    Position Offset Coordinate Space=::Local
    UsePositionOffset=true
    Write Material Random=true
   )
   /SetVariables(Particles.VisibilityTag=/Niagara/DynamicInputs/Helpers/MaskInt(
     Mask Source:NiagaraInt32=1
     Mask Parameter:NiagaraInt32=$System.intVariant
     Mask:NiagaraInt32=0
     Masked Int:NiagaraInt32=0
     Passthrough Int:NiagaraInt32=1
    )) {isSetParameters=true, module=SetVariables_#5}
   &InheritVelocity(Inherited Velocity Amount Scale=<0 0 1>, Speed Limit=true, UseSpeedThreshold=false)
   &AddVelocity(
    Cone Angle=170
    Cone Axis=<0 0 1>
    Inner Cone Angle=190
    Module.UseConeFalloff=false
    Module.UseVelDistribution=false
    Override Seed=false
    Velocity Mode=::NewEnumerator2
    Velocity Speed=&Lerp_Float(A=300, B=600, Alpha=$Particles.StrengthRandom)
   )
   &InitialMeshOrientation(
    Mesh Orientation Mode=::NewEnumerator6
    Orientation Quaternion:Quat4f=$Particles.SkeletalMeshLocation.SampledRotation
    Rotation=&RandomRangeVector(Minimum=<-0.1 -0.1 0>, Maximum=<0.1 0.1 1>)
    Use Rotation Vector=true
   )
   /SetVariables(Particles.MeshIndex=&RandomRangeInt(Minimum=0, Maximum=3)) {isSetParameters=true, module=SetVariables_#6}
  ]
  particleUpdateScript [
   &ParticleState()
   &ScaleMeshSize(Scale Factor=&Multiply_VectorByFloat(
     Vector=&VectorFromCurve(
      VectorCurve=DataInterface() {propertyValues {XCurve {keys [<at 0 1>,<at 1 1>]}, YCurve {keys [<at 0 1>,<at 1 1>]}, ZCurve {keys [<at 0 0.1>,<at 0.4 1>]}}}
      CurveIndex=$Particles.Age
     )
     Float=&FloatFromCurve(FloatCurve=DataInterface() {propertyValues {Curve {
         defaultValue=0
         keys [
          {arriveTangent=9.0725, interpMode=RCIM_Cubic, leaveTangent=9.0725, tangentMode=RCTM_User, time=0, value=0}
          {arriveTangent=0.3041, interpMode=RCIM_Cubic, leaveTangent=0.3041, tangentMode=RCTM_User, time=0.1, value=0.9847}
          {arriveTangent=-1.0941, interpMode=RCIM_Cubic, leaveTangent=-1.0941, time=0.7473, value=0.9144}
          {arriveTangent=-0.0405, interpMode=RCIM_Cubic, leaveTangent=-0.0405, tangentMode=RCTM_User, time=1, value=0}
         ]
        }}})
    ))
   &CurlNoiseForce(
    Bias Noise=false
    Noise Frequency=0.03
    Noise Quality / Cost=::NewEnumerator0
    Noise Strength=&RandomRangeFloat(Minimum=50, Maximum=100)
    Pan Noise=false
   )
   &Drag(
    Drag=&FloatFromCurve(
     FloatCurve=DataInterface() {propertyValues {Curve {keys [<at 0.5 2.5>,<at 0.6 4.5>]}}}
     CurveIndex=&Add_Float(
      A=$Particles.NormalizedAge
      B=&FloatFromCurve(
       FloatCurve=DataInterface() {propertyValues {Curve {
          defaultValue=1
          keys [
           {arriveTangent=0.9605, interpMode=RCIM_Cubic, leaveTangent=0.9605, tangentMode=RCTM_User, time=0, value=0}
           {arriveTangent=0.6307, interpMode=RCIM_Cubic, leaveTangent=0.6307, tangentMode=RCTM_User, time=1, value=1}
          ]
         }}}
       CurveIndex=&RandomRangeFloat(Minimum=0, Maximum=1)
       Scale Curve=0.6
      )
     )
    )
    Use Linear Drag=true
    Use Rotational Drag=true
   )
   &UpdateMeshOrientation(
    Rotation Rate=&FloatFromCurve(
     FloatCurve=DataInterface() {propertyValues {Curve {keys [{arriveTangent=3.1971, interpMode=RCIM_Cubic, leaveTangent=3.1971, tangentMode=RCTM_User, time=0, value=1.5}, <at 1 4 RCIM_Cubic>]}}}
     CurveIndex=&Divide_Float(A=&VectorLength(Vector=$Particles.Velocity), B=600)
    )
    Rotation Vector=&RandomVector(Vector Scale=1)
   )
   &InheritSourceMovement(
    Applied Position Scale=&VectorFromCurve(VectorCurve=DataInterface() {propertyValues {
       XCurve {keys [<at 0.3287 1 RCIM_Cubic>,<at 1 0 RCIM_Cubic>]}
       YCurve {keys [<at 0.3287 1 RCIM_Cubic>,<at 1 0 RCIM_Cubic>]}
       ZCurve {keys [<at 0.5 0.5>]}
      }})
    ApplyForce=false
    ApplyPosition=true
    ApplyPositionFalloff=false
   )
   &SolveForcesAndVelocity(Clamp Velocity=false, Limit Acceleration=false)
  ]
  rendererClasses [&NiagaraMeshRendererProperties]
  renderers [
   {rendererClass=&NiagaraMeshRendererProperties, rendererIndex=0}
   {rendererClass=&NiagaraMeshRendererProperties, rendererIndex=1}
  ]
 }
 Notes_Up {
  emitterUpdateScript [
   &EmitterState(
    Inactive Response=::NewEnumerator1
    Life Cycle Mode=::NewEnumerator1
    Loop Behavior=::NewEnumerator1
    Loop Delay=&FN_FrontEndFloatModifier(Gameplay=0.95, Front End=0.7)
    Loop Duration=2.5
    UseLoopDelay=true
   )
   &SpawnBurst_Instantaneous(
    Spawn Count=&RandomRangeInt(Minimum=6, Maximum=8)
    Use Spawn Probability=false
    UseLoopCountLimit=false
   )
   &SpawnBurst_Instantaneous(
    Spawn Count=&RandomRangeInt(Minimum=8, Maximum=12)
    Spawn Time=0.4
    Use Spawn Probability=true
    UseLoopCountLimit=false
   ) {module=SpawnBurst_Instantaneous001}
  ]
  particleSpawnScript [
   /SetVariables(
    Particles.CircleRandom:NiagaraFloat=&RandomRangeFloat(Minimum=0, Maximum=1)
    Particles.StrengthRandom:NiagaraFloat=&FloatFromCurve(
     FloatCurve=DataInterface() {propertyValues {Curve {keys [
         {arriveTangent=1.5724, interpMode=RCIM_Cubic, leaveTangent=1.5724, tangentMode=RCTM_User, time=0, value=0}
         {arriveTangent=1.2701, interpMode=RCIM_Cubic, leaveTangent=1.2701, tangentMode=RCTM_User, time=0.4337, value=0.6853}
         <at 1 1 RCIM_Cubic>
        ]}}}
     CurveIndex=&RandomRangeFloat(Minimum=0.4, Maximum=1)
    )
   ) {isSetParameters=true, module=SetVariables_#4}
   &InitializeParticle(
    Color=&MaskLinearColor(
     Mask Source=1
     Mask Parameter=$System.intVariant
     Mask=0
     Masked Linear Color=&SelectLinearColorFromArray(
      Array Sampling Mode=::NewEnumerator0
      Direct Array Index=?
      Color Selection Array=DataInterface() {propertyValues {ColorData [<rgba 0.36 0.9 0.5535 1>, <rgba 0.8019 0.5 1 1>, <rgba 1 0.5127 0.3 1>]}}
      Random Seed=0
     )
     Passthrough Linear Color=&SelectLinearColorFromArray(
      Array Sampling Mode=::NewEnumerator1
      Direct Array Index=$System.intVariant
      Color Selection Array=DataInterface() {propertyValues {ColorData [<rgba 1 0.5127 0.3 1>, <rgba 1 0.7061 0.3217 1>, <rgba 0.2321 0.2437 0.2696 1>]}}
      Random Seed=?
     )
    )
    Lifetime=&RandomRangeFloat(Minimum=0.7, Maximum=1.1)
    Mesh Renderer Array Visibility Mode=::NewEnumerator3
    Mesh Renderer Info=DataInterface() {propertyValues {MeshRenderer=&self:Notes_Up_1.NiagaraMeshRendererProperties_1}}
    Mesh Scale Mode=::NewEnumerator3
    Position Offset=&FN_FrontEndVector3Modifier(Gameplay=<0 0 0>, Front End=<0 -40 80>)
    Position Offset Coordinate Space=::Local
    UsePositionOffset=true
    Write Material Random=true
   )
   /SetVariables(Particles.VisibilityTag=&MaskInt(
     Mask Source:NiagaraInt32=1
     Mask Parameter:NiagaraInt32=$System.intVariant
     Mask:NiagaraInt32=0
     Masked Int:NiagaraInt32=0
     Passthrough Int:NiagaraInt32=1
    )) {isSetParameters=true}
   &ShapeLocation(
    MaskSpawnGroup=false
    Override Seed=false
    Ring / Disc Distribution Mode:ENiagaraSphereDistributionMode=::NewEnumerator1
    Ring Radius=26
    Shape Primitive=::NewEnumerator4
    U Position=$Particles.CircleRandom
    UseOwnerScale=true
   )
   &InheritVelocity(Inherited Velocity Amount Scale=<0 0 1>, Speed Limit=true, UseSpeedThreshold=false)
   &AddVelocity(
    Cone Angle=65
    Cone Axis=<0 0 1>
    Inner Cone Angle=50
    Module.UseConeFalloff=false
    Module.UseVelDistribution=false
    Override Seed=false
    Velocity Mode=::NewEnumerator2
    Velocity Speed=&Lerp_Float(A=200, B=600, Alpha=$Particles.StrengthRandom)
   ) {module=AddVelocity001}
   &InitialMeshOrientation(
    Mesh Orientation Mode=::NewEnumerator6
    Orientation Quaternion:Quat4f=$Particles.SkeletalMeshLocation.SampledRotation
    Rotation=&RandomRangeVector(Minimum=<-0.1 -0.1 0>, Maximum=<0.1 0.1 1>)
    Use Rotation Vector=true
   )
   /SetVariables(Particles.MeshIndex=&RandomRangeInt(Minimum=0, Maximum=3)) {isSetParameters=true, module=SetVariables_#7}
  ]
  particleUpdateScript [
   &ParticleState()
   &ScaleMeshSize(Scale Factor=&Multiply_VectorByFloat(
     Vector=&VectorFromCurve(
      VectorCurve=DataInterface() {propertyValues {XCurve {keys [<at 0 1>,<at 1 1>]}, YCurve {keys [<at 0 1>,<at 1 1>]}, ZCurve {keys [<at 0 0.1>,<at 0.4 1>]}}}
      CurveIndex=$Particles.Age
     )
     Float=&FloatFromCurve(FloatCurve=DataInterface() {propertyValues {Curve {
         defaultValue=0
         keys [
          {arriveTangent=9.0725, interpMode=RCIM_Cubic, leaveTangent=9.0725, tangentMode=RCTM_User, time=0, value=0}
          {arriveTangent=0.3041, interpMode=RCIM_Cubic, leaveTangent=0.3041, tangentMode=RCTM_User, time=0.1, value=0.9847}
          {arriveTangent=-1.0941, interpMode=RCIM_Cubic, leaveTangent=-1.0941, time=0.7473, value=0.9144}
          {arriveTangent=-0.0405, interpMode=RCIM_Cubic, leaveTangent=-0.0405, tangentMode=RCTM_User, time=1, value=0}
         ]
        }}})
    ))
   &GravityForce(Gravity=&Multiply_VectorByFloat(
     Vector=&RandomRangeVector(Minimum=<0 0 -700>, Maximum=<0 0 -1000>)
     Float=&FloatFromCurve(
      FloatCurve=DataInterface() {propertyValues {Curve {
         defaultValue=1
         keys [
          {arriveTangent=9.2745, interpMode=RCIM_Cubic, leaveTangent=9.2745, tangentMode=RCTM_User, time=0, value=0}
          {interpMode=RCIM_Cubic, tangentMode=RCTM_User, time=0.2876, value=1}
          {arriveTangent=0.0017, tangentMode=RCTM_User, time=1, value=0.5299}
         ]
        }}}
      Scale Curve=$Particles.StrengthRandom
     )
    ))
   &CurlNoiseForce(
    Bias Noise=false
    Noise Frequency=0.03
    Noise Quality / Cost=::NewEnumerator0
    Noise Strength=&RandomRangeFloat(Minimum=50, Maximum=100)
    Pan Noise=false
   )
   &Drag(Use Linear Drag=true, Use Rotational Drag=true)
   &UpdateMeshOrientation(
    Rotation Rate=&FloatFromCurve(
     FloatCurve=DataInterface() {propertyValues {Curve {keys [<at 0 1.5 RCIM_Cubic>,<at 1 4 RCIM_Cubic>]}}}
     CurveIndex=&Divide_Float(A=&VectorLength(Vector=$Particles.Velocity), B=600)
    )
    Rotation Vector=&RandomVector(Vector Scale=1)
   )
   &InheritSourceMovement(
    Applied Position Scale=&VectorFromCurve(VectorCurve=DataInterface() {propertyValues {
       XCurve {keys [<at 0.3287 1 RCIM_Cubic>,<at 1 0 RCIM_Cubic>]}
       YCurve {keys [<at 0.3287 1 RCIM_Cubic>,<at 1 0 RCIM_Cubic>]}
       ZCurve {keys [<at 0.5 0.3>]}
      }})
    ApplyForce=false
    ApplyPosition=true
    ApplyPositionFalloff=false
   )
   &SolveForcesAndVelocity(Clamp Velocity=false, Limit Acceleration=false)
  ]
  rendererClasses [&NiagaraMeshRendererProperties]
  renderers [
   {rendererClass=&NiagaraMeshRendererProperties, rendererIndex=0}
   {rendererClass=&NiagaraMeshRendererProperties, rendererIndex=1}
  ]
 }
 Notes_Wings {
  emitterUpdateScript [
   &EmitterState(
    Inactive Response=::NewEnumerator1
    Life Cycle Mode=::NewEnumerator1
    Loop Behavior=::NewEnumerator1
    Loop Delay=&FN_FrontEndFloatModifier(Gameplay=0.95, Front End=0.7)
    Loop Duration=2.5
    UseLoopDelay=true
   )
   &SpawnBurst_Instantaneous(Spawn Count=8, Spawn Time=0.4, Use Spawn Probability=false, UseLoopCountLimit=false)
   &SpawnBurst_Instantaneous(Spawn Count=8, Spawn Group=1, Spawn Time=0.45, Use Spawn Probability=false, UseLoopCountLimit=false) {module=SpawnBurst_Instantaneous001}
   &SpawnBurst_Instantaneous(Spawn Count=4, Spawn Group=2, Spawn Time=0.5, Use Spawn Probability=false, UseLoopCountLimit=false) {module=SpawnBurst_Instantaneous002}
   &SpawnBurst_Instantaneous(Spawn Count=4, Spawn Group=3, Spawn Time=0.57, Use Spawn Probability=false, UseLoopCountLimit=false) {module=SpawnBurst_Instantaneous003}
  ]
  particleSpawnScript [
   /SetVariables(Particles.StrengthRandom:NiagaraFloat=&RandomRangeFloat(Minimum=0.75, Maximum=1)) {isSetParameters=true, module=SetVariables_#4}
   &PartitionParticles(Number Of Partitions=3)
   &InitializeParticle(
    Color=&MaskLinearColor(
     Mask Source=1
     Mask Parameter=$System.intVariant
     Mask=0
     Masked Linear Color=&SelectLinearColorFromArray(
      Array Sampling Mode=::NewEnumerator0
      Direct Array Index=?
      Color Selection Array=DataInterface() {propertyValues {ColorData [<rgba 0.36 0.9 0.5535 1>, <rgba 0.8019 0.5 1 1>, <rgba 1 0.5127 0.3 1>]}}
      Random Seed=0
     )
     Passthrough Linear Color=&SelectLinearColorFromArray(
      Array Sampling Mode=::NewEnumerator1
      Direct Array Index=$System.intVariant
      Color Selection Array=DataInterface() {propertyValues {ColorData [<rgba 1 0.5127 0.3 1>, <rgba 1 0.7061 0.3217 1>, <rgba 0.2321 0.2437 0.2696 1>]}}
      Random Seed=?
     )
    )
    Lifetime=&RandomRangeFloat(Minimum=0.6, Maximum=0.9)
    Mesh Renderer Array Visibility Mode=::NewEnumerator3
    Mesh Renderer Info=DataInterface() {propertyValues {MeshRenderer=&self:Notes_Wings_0.NiagaraMeshRendererProperties_1}}
    Mesh Scale Mode=::NewEnumerator3
    Position Offset=&FN_FrontEndVector3Modifier(Gameplay=<0 0 900>, Front End=<0 -30 100>)
    UsePositionOffset=false
    Write Material Random=true
   )
   /SetVariables(Particles.VisibilityTag=&MaskInt(
     Mask Source:NiagaraInt32=1
     Mask Parameter:NiagaraInt32=$System.intVariant
     Mask:NiagaraInt32=0
     Masked Int:NiagaraInt32=0
     Passthrough Int:NiagaraInt32=1
    )) {isSetParameters=true, module=SetVariables_#5}
   &SocketLocation(
    SamplePosOffset=false
    Socket Index=&Add_Integer(
     A=$Particles.PartitionParticles.Partition
     B=&Add_Integer(
      A=1
      B=&SelectIntFromArray(Direct Array Index=&SpawnGroup(), Int Selection Array=DataInterface() {propertyValues {IntData [0,3,6,9]}})
     )
    )
    Socket Reader=$System.Socket Reader
    Socket Sampling Mode=::NewEnumerator3
    VelocitySpeedLimit=false
   )
   &ShapeLocation(
    Additional Quaternion Rotation=false
    Box Size=<1 1 1>
    MaskSpawnGroup=false
    Non Uniform Scale=&Multiply_VectorByFloat(
     Vector=&SelectVectorFromArray(
      Direct Array Index=&Add_Integer(
       A=$Particles.PartitionParticles.Partition
       B=&SelectIntFromArray(Direct Array Index=&SpawnGroup(), Int Selection Array=DataInterface() {propertyValues {IntData [0,3,6,9]}})
      )
      Vector Selection Array=DataInterface() {propertyValues {FloatData [<66 60 0>, <46 70 0>, <45 100 0>, <66 60 0>, <46 70 0>, <45 100 0>, <26 0 42>, <30 0 42>, <23.2 0 42>, <26 0 42>, <30 0 42>, <23.2 0 42>]}}
     )
     Float=1
    )
    Override Seed=false
    Rotation Mode=::NewEnumerator3
    Rotation Quaternion=$Output.SocketLocation.SampledRotation
    Shape Primitive=::NewEnumerator2
    Surface Only=false
    UseOwnerScale=true
   )
   &InheritVelocity(Inherited Velocity Amount Scale=<0 0 1>, Speed Limit=true, UseSpeedThreshold=false)
   &AddVelocity(
    Velocity=&Add_Vector(A=<0 0 0.25>, B=&RandomVector(Vector Scale=0.75))
    Velocity Speed Scale=&RandomRangeFloat(Minimum=222, Maximum=444)
   )
   &AddVelocity(Module.Use Falloff Distance=false, Override Seed=false, Velocity Mode=::NewEnumerator1, Velocity Speed=222) {module=AddVelocity001}
   &InitialMeshOrientation(
    Mesh Orientation Mode=::NewEnumerator6
    Orientation Quaternion:Quat4f=$Particles.SkeletalMeshLocation.SampledRotation
    Rotation=&RandomRangeVector(Minimum=<-0.1 -0.1 0>, Maximum=<0.1 0.1 1>)
    Use Rotation Vector=true
   )
   /SetVariables(Particles.MeshIndex=&RandomRangeInt(Minimum=0, Maximum=3)) {isSetParameters=true, module=SetVariables_#6}
  ]
  particleUpdateScript [
   &ParticleState()
   &ScaleMeshSize(Scale Factor=&Multiply_VectorByFloat(
     Vector=&VectorFromCurve(
      VectorCurve=DataInterface() {propertyValues {XCurve {keys [<at 0 1>,<at 1 1>]}, YCurve {keys [<at 0 1>,<at 1 1>]}, ZCurve {keys [<at 0 0.1>,<at 0.4 1>]}}}
      CurveIndex=$Particles.Age
     )
     Float=&FloatFromCurve(FloatCurve=DataInterface() {propertyValues {Curve {
         defaultValue=0
         keys [
          {arriveTangent=9.0725, interpMode=RCIM_Cubic, leaveTangent=9.0725, tangentMode=RCTM_User, time=0, value=0}
          {arriveTangent=0.3041, interpMode=RCIM_Cubic, leaveTangent=0.3041, tangentMode=RCTM_User, time=0.1, value=0.9847}
          {arriveTangent=-1.0941, interpMode=RCIM_Cubic, leaveTangent=-1.0941, time=0.7473, value=0.9144}
          {arriveTangent=-0.0405, interpMode=RCIM_Cubic, leaveTangent=-0.0405, tangentMode=RCTM_User, time=1, value=0}
         ]
        }}})
    ))
   &CurlNoiseForce(
    Bias Noise=false
    Noise Frequency=0.03
    Noise Quality / Cost=::NewEnumerator0
    Noise Strength=&RandomRangeFloat(Minimum=50, Maximum=100)
    Pan Noise=false
   )
   &Drag(
    Drag=&FloatFromCurve(
     FloatCurve=DataInterface() {propertyValues {Curve {keys [<at 0.5 2>,<at 0.6 6>]}}}
     CurveIndex=&Add_Float(
      A=$Particles.NormalizedAge
      B=&FloatFromCurve(
       FloatCurve=DataInterface() {propertyValues {Curve {
          defaultValue=1
          keys [
           {arriveTangent=0.9605, interpMode=RCIM_Cubic, leaveTangent=0.9605, tangentMode=RCTM_User, time=0, value=0}
           {arriveTangent=0.6307, interpMode=RCIM_Cubic, leaveTangent=0.6307, tangentMode=RCTM_User, time=1, value=1}
          ]
         }}}
       CurveIndex=&RandomRangeFloat(Minimum=0, Maximum=1)
       Scale Curve=0.6
      )
     )
    )
    Use Linear Drag=true
    Use Rotational Drag=true
   )
   &UpdateMeshOrientation(
    Rotation Rate=&FloatFromCurve(
     FloatCurve=DataInterface() {propertyValues {Curve {keys [{arriveTangent=3.1971, interpMode=RCIM_Cubic, leaveTangent=3.1971, tangentMode=RCTM_User, time=0, value=1.5}, <at 1 4 RCIM_Cubic>]}}}
     CurveIndex=&Divide_Float(A=&VectorLength(Vector=$Particles.Velocity), B=600)
    )
    Rotation Vector=&RandomVector(Vector Scale=1)
   )
   &InheritSourceMovement(
    Applied Position Scale=&VectorFromCurve(VectorCurve=DataInterface() {propertyValues {
       XCurve {keys [<at 0.3287 1 RCIM_Cubic>,<at 1 0 RCIM_Cubic>]}
       YCurve {keys [<at 0.3287 1 RCIM_Cubic>,<at 1 0 RCIM_Cubic>]}
       ZCurve {keys [<at 0.5 0.3>]}
      }})
    ApplyForce=false
    ApplyPosition=true
    ApplyPositionFalloff=false
   )
   &SolveForcesAndVelocity(Clamp Velocity=false, Limit Acceleration=false)
  ]
  rendererClasses [&NiagaraMeshRendererProperties]
  renderers [
   {rendererClass=&NiagaraMeshRendererProperties, rendererIndex=0}
   {rendererClass=&NiagaraMeshRendererProperties, rendererIndex=1}
  ]
 }
 Sparkles {
  emitterUpdateScript [
   &EmitterState(
    Inactive Response=::NewEnumerator1
    Life Cycle Mode=::NewEnumerator1
    Loop Behavior=::NewEnumerator1
    Loop Delay=&FN_FrontEndFloatModifier(Gameplay=0.95, Front End=0.7)
    Loop Duration=2.5
    UseLoopDelay=true
   )
   &SpawnBurst_Instantaneous(
    Spawn Count=&RandomRangeInt(Minimum=10, Maximum=18)
    Spawn Probability=&MaskFloat(Mask Source=1, Mask Parameter=$System.intVariant, Mask=0, Masked Float=0, Passthrough Float=1)
    Use Spawn Probability=true
    UseLoopCountLimit=false
   )
   &SpawnBurst_Instantaneous(
    Spawn Count=&RandomRangeInt(Minimum=12, Maximum=22)
    Spawn Probability=&MaskFloat(Mask Source=1, Mask Parameter=$System.intVariant, Mask=0, Masked Float=0, Passthrough Float=1)
    Spawn Time=0.4
    Use Spawn Probability=true
    UseLoopCountLimit=false
   ) {module=SpawnBurst_Instantaneous001}
  ]
  particleSpawnScript [
   /SetVariables(
    Particles.CircleRandom:NiagaraFloat=&RandomRangeFloat(Minimum=0, Maximum=1)
    Particles.StrengthRandom:NiagaraFloat=&FloatFromCurve(
     FloatCurve=DataInterface() {propertyValues {Curve {keys [
         {arriveTangent=1.5724, interpMode=RCIM_Cubic, leaveTangent=1.5724, tangentMode=RCTM_User, time=0, value=0}
         {arriveTangent=1.2701, interpMode=RCIM_Cubic, leaveTangent=1.2701, tangentMode=RCTM_User, time=0.4337, value=0.6853}
         <at 1 1 RCIM_Cubic>
        ]}}}
     CurveIndex=&RandomRangeFloat(Minimum=0.4, Maximum=1)
    )
   ) {isSetParameters=true, module=SetVariables_#4}
   &InitializeParticle(
    Color=&MaskLinearColor(
     Mask Source=1
     Mask Parameter=$System.intVariant
     Mask=1
     Masked Linear Color=<rgba 1 0.7253 0.169 0.5217>
     Passthrough Linear Color=<rgba 0.8448 0.9614 1 0.5652>
    )
    Lifetime=&RandomRangeFloat(Minimum=0.7, Maximum=1.1)
    Position Offset=&FN_FrontEndVector3Modifier(Gameplay=<0 0 20>, Front End=<0 -30 100>)
    Position Offset Coordinate Space=::Local
    Sprite Size Mode=::NewEnumerator3
    Uniform Sprite Size=&FloatFromCurve(
     FloatCurve=DataInterface() {propertyValues {Curve {keys [<at 0 3 RCIM_Cubic>, {arriveTangent=3.5447, interpMode=RCIM_Cubic, leaveTangent=3.5447, tangentMode=RCTM_User, time=1, value=5}]}}}
     CurveIndex=&RandomRangeFloat(Minimum=0, Maximum=1)
     Scale Curve=3
    )
    UsePositionOffset=true
    Write Material Random=true
   )
   &InheritVelocity(Inherited Velocity Amount Scale=<0 0 1>, Speed Limit=true, UseSpeedThreshold=false)
   &AddVelocity(
    Velocity=&Add_Vector(A=&RandomVector(Vector Scale=1), B=<0 0 0.5>)
    Velocity Speed Scale=&RandomRangeFloat(Minimum=333, Maximum=444)
   ) {module=AddVelocity001}
   /SetVariables(Particles.MeshIndex=&RandomRangeInt(Minimum=0, Maximum=3)) {isSetParameters=true, module=SetVariables_#7}
  ]
  particleUpdateScript [
   &ParticleState()
   &ScaleSpriteSize(Scale Factor=&Vector2DFromFloat(Value=&FloatFromCurve(
      FloatCurve=DataInterface() {propertyValues {Curve {keys [
          <at 0 0 RCIM_Cubic>
          {arriveTangent=-1.0005, interpMode=RCIM_Cubic, leaveTangent=-1.0005, tangentMode=RCTM_User, time=0.2894, value=1}
          {arriveTangent=0.1497, interpMode=RCIM_Cubic, leaveTangent=0.1497, tangentMode=RCTM_User, time=0.5815, value=0.4733}
          {arriveTangent=-0.8218, tangentMode=RCTM_User, time=1.0011, value=0.002}
         ]}}}
      Scale Curve=&FloatFromCurve(
       FloatCurve=DataInterface() {propertyValues {Curve {keys [<at 0 3>,<at 0.2 1.2>,<at 1 1>]}}}
       CurveIndex=&RandomRangeFloat(Minimum=0, Maximum=1, Evaluation Type=::NewEnumerator1)
      )
     )))
   &GravityForce(Gravity=&Multiply_VectorByFloat(
     Vector=&RandomRangeVector(Minimum=<0 0 -200>, Maximum=<0 0 -300>)
     Float=&FloatFromCurve(
      FloatCurve=DataInterface() {propertyValues {Curve {
         defaultValue=1
         keys [
          {arriveTangent=9.2745, interpMode=RCIM_Cubic, leaveTangent=9.2745, tangentMode=RCTM_User, time=0, value=0}
          {interpMode=RCIM_Cubic, tangentMode=RCTM_User, time=0.2876, value=1}
          {arriveTangent=0.0017, tangentMode=RCTM_User, time=1, value=0.5299}
         ]
        }}}
      Scale Curve=$Particles.StrengthRandom
     )
    ))
   &Drag(Drag=2, Use Linear Drag=true, Use Rotational Drag=true)
   &InheritSourceMovement(
    Applied Position Scale=&VectorFromCurve(VectorCurve=DataInterface() {propertyValues {
       XCurve {keys [<at 0.3287 1 RCIM_Cubic>,<at 1 0 RCIM_Cubic>]}
       YCurve {keys [<at 0.3287 1 RCIM_Cubic>,<at 1 0 RCIM_Cubic>]}
       ZCurve {keys [<at 0.5 0.3>]}
      }})
    ApplyForce=false
    ApplyPosition=true
    ApplyPositionFalloff=false
   )
   &SolveForcesAndVelocity(Clamp Velocity=false, Limit Acceleration=false)
  ]
  rendererClasses [&NiagaraSpriteRendererProperties]
  renderers [{rendererClass=&NiagaraSpriteRendererProperties}]
 }
}
summary {
 systemName=NS_Glider_FunkyFeather_Deploy
 emitters [
  {emitterName=Notes_Up, rendererClasses [&NiagaraMeshRendererProperties]}
  {emitterName=Notes_Out, rendererClasses [&NiagaraMeshRendererProperties]}
  {emitterName=Notes_Wings, rendererClasses [&NiagaraMeshRendererProperties]}
  {emitterName=Sparkles, rendererClasses [&NiagaraSpriteRendererProperties]}
 ]
 userVariables [
  User.bIsFrontEnd {type=NiagaraBool}
  User.bIsFrontEndPreview {type=NiagaraBool}
  User.bIsFullyDeployed {type=NiagaraBool}
  User.ParaGlideLeanAlpha {type=NiagaraFloat, defaultValue=1, description=""}
  User.Variant {type=NiagaraFloat, defaultValue=0, description=""}
 ]
}
