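# Niagara trail effect for the FunkyFeather glider, combining drifting mesh notes, sparkles, speed lines and green and brown wing ribbons.
# Notes and sparkles spawn in shapes around partition-selected sockets, with sparkles staying socket-attached, while ribbons use anim-trail socket pairs and time-of-day tinting.
# An integer variant selects colors and masks elements, disabling sparkles for variant zero.
# Front-end switches adjust motion and rates, and gameplay speed-line opacity follows ForwardDot.
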NiagaraSystem /BRCosmetics/Gadgets/Assets/VinderTech_GliderChute/Glider_FunkyFeather_Cap/FX/NS_Glider_FunkyFeather_Trail
dependencies {
 usedDataInterfaces [
  &NiagaraDataInterfaceArrayColor
  &NiagaraDataInterfaceMeshRendererInfo
  &NiagaraDataInterfaceArrayFloat3
  &NiagaraDataInterfaceVectorCurve
  &NiagaraDataInterfaceCurve
  &NiagaraDataInterfaceVector2DCurve
  &NiagaraDataInterfaceColorCurve
  &NiagaraDataInterfaceArrayFloat
  &NiagaraDataInterfaceSocketReader
 ]
 usedDynamicInputs [
  &Vector2DFromCurve
  &SetBoolByFloatComparison
  &RemapRange
  &NormalizeVector
  &TransformVector
 ]
 usedModules [
  &SystemState
  &self:SystemSpawnScript.SystemScriptSource.SystemScriptGraph.NiagaraNodeAssignment_2.SetVariables_#1
  &Timeline
  &FN_Integrator
  &self:SystemSpawnScript.SystemScriptSource.SystemScriptGraph.NiagaraNodeAssignment_3.SetVariables_#2
  &self:SystemSpawnScript.SystemScriptSource.SystemScriptGraph.NiagaraNodeAssignment_1.SetVariables_#3
  &self:SystemSpawnScript.SystemScriptSource.SystemScriptGraph.NiagaraNodeAssignment_0.SetVariables_#4
 ]
}
emitters {
 Notes {
  emitterUpdateScript [
   &EmitterState()
   &SpawnRate(
    Spawn Probability=&FN_FrontEndFloatModifier(Gameplay=$System.Timeline.Time, Front End=1)
    SpawnRate=24
    Use Spawn Probability=true
   )
  ]
  particleSpawnScript [
   /SetVariables(Particles.PushForce:NiagaraFloat=&RandomRangeFloat(Minimum=0, Maximum=1)) {isSetParameters=true}
   &PartitionParticles(Number Of Partitions=12)
   /SetVariables(Particles.VisibilityTag=/Niagara/DynamicInputs/Helpers/MaskInt(
     Mask Source:NiagaraInt32=1
     Mask Parameter:NiagaraInt32=$System.intVariant
     Mask:NiagaraInt32=0
     Masked Int:NiagaraInt32=0
     Passthrough Int:NiagaraInt32=1
    )) {isSetParameters=true}
   &InitializeParticle(
    Color=&MaskLinearColor(
     Mask Source=1
     Mask Parameter=$System.intVariant
     Mask=0
     Masked Linear Color=&SelectLinearColorFromArray(
      Array Sampling Mode=::NewEnumerator0
      Direct Array Index=?
      Color Selection Array=DataInterface() {propertyValues {ColorData [<rgba 0.36 0.9 0.5535 1>, <rgba 0.8019 0.5 1 1>, <rgba 1 0.5127 0.3 1>]}}
      Random Seed=0
     )
     Passthrough Linear Color=&SelectLinearColorFromArray(
      Array Sampling Mode=::NewEnumerator1
      Direct Array Index=$System.intVariant
      Color Selection Array=DataInterface() {propertyValues {ColorData [<rgba 1 0.5127 0.3 1>, <rgba 1 0.7061 0.3217 1>, <rgba 0.2321 0.2437 0.2696 1>]}}
      Random Seed=?
     )
    )
    Lifetime=&RandomRangeFloat(Minimum=0.7, Maximum=1.25)
    Mesh Scale Mode=::NewEnumerator3
    UsePositionOffset=true
    Write Material Random=true
   )
   &InitialMeshOrientation(Use Rotation Vector=true)
   &CameraOffset(Camera Offset Amount=12)
   &SocketLocation(
    SamplePosOffset=false
    Socket Index=&Add_Integer(A=$Particles.PartitionParticles.Partition, B=8)
    Socket Reader=$System.Socket Reader
    Socket Sampling Mode=::NewEnumerator3
    VelocitySpeedLimit=false
   )
   &ShapeLocation(
    Additional Quaternion Rotation=false
    Box Size=<1 1 1>
    MaskSpawnGroup=false
    Non Uniform Scale=&Multiply_VectorByFloat(
     Vector=&SelectVectorFromArray(
      Direct Array Index=$Particles.PartitionParticles.Partition
      Vector Selection Array=DataInterface() {propertyValues {FloatData [
         <66 60 0>
         <66 60 0>
         <46 70 0>
         <46 70 0>
         <45 100 0>
         <45 100 0>
         <26 0 42>
         <26 0 42>
         <30 0 42>
         <30 0 42>
         <23.2 0 42>
         <23.2 0 42>
         <0 0 0>
         <0 0 0>
        ]}}
     )
     Float=0.75
    )
    Override Seed=false
    Rotation Mode=::NewEnumerator3
    Rotation Quaternion=$Output.SocketLocation.SampledRotation
    Shape Primitive=::NewEnumerator2
    Surface Only=false
    UseOwnerScale=true
   )
   &AddVelocity(Velocity=&Multiply_VectorByFloat(
     Vector=&FN_FrontEndVector3Modifier(Gameplay=<0 40 0>, Front End=<0 -180 0>)
     Float=$Particles.PushForce
    ))
  ]
  particleUpdateScript [
   &ParticleState()
   &ScaleMeshSize(Scale Factor=&Multiply_VectorByFloat(
     Vector=&VectorFromCurve(
      VectorCurve=DataInterface() {propertyValues {XCurve {keys [<at 0 1>,<at 1 1>]}, YCurve {keys [<at 0 1>,<at 1 1>]}, ZCurve {keys [<at 0 0.1>,<at 0.4 1>]}}}
      CurveIndex=$Particles.Age
     )
     Float=&FloatFromCurve(FloatCurve=DataInterface() {propertyValues {Curve {
         defaultValue=0
         keys [
          {arriveTangent=9.0725, interpMode=RCIM_Cubic, leaveTangent=9.0725, tangentMode=RCTM_User, time=0, value=0}
          {arriveTangent=0.3041, interpMode=RCIM_Cubic, leaveTangent=0.3041, tangentMode=RCTM_User, time=0.2053, value=0.9847}
          {arriveTangent=-1.239, interpMode=RCIM_Cubic, leaveTangent=-1.239, time=0.7473, value=0.9144}
          {arriveTangent=-0.0405, interpMode=RCIM_Cubic, leaveTangent=-0.0405, tangentMode=RCTM_User, time=1, value=0}
         ]
        }}})
    ))
   &CurlNoiseForce(
    Bias Noise=false
    Noise Frequency=1
    Noise Quality / Cost=::NewEnumerator0
    Noise Strength=&RandomRangeFloat(Minimum=20, Maximum=200)
    Pan Noise=false
   )
   &UpdateMeshOrientation(
    Rotation Rate=&Lerp_Float(A=1, B=3, Alpha=$Particles.Initial.PushForce)
    Rotation Vector=&RandomVector(Vector Scale=1)
   )
   &InheritSourceMovement(
    Applied Position Scale=&VectorFromCurve(VectorCurve=DataInterface() {propertyValues {
       XCurve {keys [<at 0 1 RCIM_Cubic>,<at 1 0 RCIM_Cubic>]}
       YCurve {keys [<at 0 1 RCIM_Cubic>,<at 1 0 RCIM_Cubic>]}
       ZCurve {keys [<at 0 1>,<at 0.5 1>,<at 1 0.5>]}
      }})
    ApplyForce=false
    ApplyPosition=true
    ApplyPositionFalloff=false
   )
   &SolveForcesAndVelocity(Clamp Velocity=false, Limit Acceleration=false)
  ]
  rendererClasses [&NiagaraMeshRendererProperties]
  renderers [
   {rendererClass=&NiagaraMeshRendererProperties, rendererIndex=0}
   {rendererClass=&NiagaraMeshRendererProperties, rendererIndex=1}
  ]
 }
 Sparkle003 {
  emitterUpdateScript [
   &EmitterState(Life Cycle Mode=::NewEnumerator1, Loop Duration=0.5, UseLoopDelay=false)
   &SpawnRate(
    SpawnRate=&MaskFloat(Mask Source=1, Mask Parameter=$System.intVariant, Mask=0, Masked Float=0, Passthrough Float=30)
    Use Spawn Probability=false
   )
  ]
  particleSpawnScript [
   &PartitionParticles(Number Of Partitions=22)
   /SetVariables(Particles.RemappedPartition:NiagaraInt32=&MakeIntFromFloat(Float=&FloatFromCurve(
      FloatCurve=DataInterface() {propertyValues {Curve {keys [<at 0 0>,<at 12.1 12>,<at 22 14>,<at 23 18>]}}}
      CurveIndex=&MakeFloatFromInt(Int=$Particles.PartitionParticles.Partition)
     ))) {isSetParameters=true}
   &InitializeParticle(
    Color=&MaskLinearColor(
     Mask Source=1
     Mask Parameter=$System.intVariant
     Mask=1
     Masked Linear Color=<rgba 1 0.7843 0.2994 0.4348>
     Passthrough Linear Color=<rgba 0.8448 0.9614 1 0.487>
    )
    Lifetime=&RandomRangeFloat(Minimum=0.6, Maximum=0.75)
    Sprite Size Mode=::NewEnumerator3
    Sprite UV Mode=::NewEnumerator2
    Uniform Sprite Size=&FloatFromCurve(
     FloatCurve=DataInterface() {propertyValues {Curve {keys [<at 0 3 RCIM_Cubic>, {arriveTangent=3.5447, interpMode=RCIM_Cubic, leaveTangent=3.5447, tangentMode=RCTM_User, time=1, value=5}]}}}
     CurveIndex=&RandomRangeFloat(Minimum=0, Maximum=1)
     Scale Curve=2.6
    )
    UsePositionOffset=false
    Write Material Random=true
   )
   &SocketLocation(
    Position Sampling=::NewEnumerator1
    SamplePosOffset=false
    Socket Index=&Add_Integer(A=$Particles.RemappedPartition, B=8)
    Socket Reader=$System.Socket Reader
    Socket Sampling Mode=::NewEnumerator3
    VelocitySpeedLimit=false
   )
   &ShapeLocation(
    Additional Quaternion Rotation=false
    Box Size=<1 1 1>
    MaskSpawnGroup=false
    Non Uniform Scale=&Multiply_VectorByFloat(
     Vector=&SelectVectorFromArray(
      Direct Array Index=$Particles.RemappedPartition
      Vector Selection Array=DataInterface() {propertyValues {FloatData [
         <66 60 0>
         <66 60 0>
         <46 70 0>
         <46 70 0>
         <44 100 0>
         <44 100 0>
         <28.6 0 42>
         <28.6 0 42>
         <33 0 42>
         <33 0 42>
         <25.52 0 42>
         <25.52 0 42>
         <0 0 0>
         <0 0 0>
        ]}}
     )
     Float=0.9
    )
    Override Seed=false
    Rotation Mode=::NewEnumerator3
    Rotation Quaternion=$Output.SocketLocation.SampledRotation
    Shape Primitive=::NewEnumerator2
    Surface Only=false
    UseOwnerScale=true
   )
   &ShapeLocation(
    Additional Quaternion Rotation=false
    MaskSpawnGroup=false
    Non Uniform Scale=&Multiply_VectorByFloat(
     Vector=&SelectVectorFromArray(
      Direct Array Index=$Particles.RemappedPartition
      Vector Selection Array=DataInterface() {propertyValues {FloatData [<0 0 0>, <0 0 0>, <0 0 0>, <0 0 0>, <0 0 0>, <0 0 0>, <0 0 0>, <0 0 0>, <0 0 0>, <0 0 0>, <0 0 0>, <0 0 0>, <40 40 21>, <21.85 26.6 24.7>]}}
     )
     Float=0.9
    )
    Override Seed=false
    Rotation Mode=::NewEnumerator3
    Rotation Quaternion=$Output.SocketLocation.SampledRotation
    Sphere Radius=1
    Sphere Surface Distribution=1
    UseOwnerScale=true
   ) {module=ShapeLocation001}
   &CameraOffset(Camera Offset Amount=17)
  ]
  particleUpdateScript [
   &ParticleState()
   &SocketLocation(
    Orientation Sampling=::NewEnumerator1
    SamplePosOffset=true
    Sampled Position Coordinate Space:ENiagaraCoordinateSpace=::Local
    Sampled Position Offset:NiagaraPosition=$Particles.Initial.Position
    Sampled Position Offset Coordinate Space:ENiagaraSkelMeshTransforms=::NewEnumerator2
    Socket Index=&Add_Integer(A=$Particles.RemappedPartition, B=8)
    Socket Reader=$System.Socket Reader
    Socket Sampling Mode=::NewEnumerator3
    VelocitySpeedLimit=false
   ) {module=SocketLocation001}
   &ScaleSpriteSize(Scale Factor=&Vector2DFromFloat(Value=&FloatFromCurve(
      FloatCurve=DataInterface() {propertyValues {Curve {keys [
          <at 0 0 RCIM_Cubic>
          {arriveTangent=-1.0005, interpMode=RCIM_Cubic, leaveTangent=-1.0005, tangentMode=RCTM_User, time=0.2894, value=1}
          {arriveTangent=0.1497, interpMode=RCIM_Cubic, leaveTangent=0.1497, tangentMode=RCTM_User, time=0.5815, value=0.4733}
          {arriveTangent=-0.8218, tangentMode=RCTM_User, time=1.0011, value=0.002}
         ]}}}
      Scale Curve=&FloatFromCurve(
       FloatCurve=DataInterface() {propertyValues {Curve {keys [<at 0 3>,<at 0.2 1.2>,<at 1 1>]}}}
       CurveIndex=&RandomRangeFloat(Minimum=0, Maximum=1, Evaluation Type=::NewEnumerator1)
      )
     )))
   &ScaleColor(
    Scale Alpha=&FloatFromCurve(
     FloatCurve=DataInterface() {propertyValues {Curve {keys [<at 0.5 1 RCIM_Cubic>,<at 1 0 RCIM_Cubic>]}}}
     Scale Curve=&MaskFloat(
      Mask Source=1
      Mask Parameter=$System.intVariant
      Mask=2
      Masked Float=1
      Passthrough Float=&SelectFloatFromArray(
       Direct Array Index=$Particles.Initial.RemappedPartition
       Float Selection Array=DataInterface() {propertyValues {FloatData [0,0,1,1,1,1,1,1,1,1,1,1,0,0]}}
      )
     )
    )
    Scale RGB=&MakeVectorFromLinearColorRGB(Color=$System.TOD Color)
    ScaleA=true
    ScaleRGB=true
   )
   &DynamicMaterialParameters(
    Index 0 Param 1=3
    Index 0 Param 2=1
    Index 0 Param 3=0.5
    Module.Index0Param0Write=true
    Module.Index0Param1Write=true
    Module.Index0Param2Write=true
    Module.Index0Param3Write=true
   )
  ]
  rendererClasses [&NiagaraSpriteRendererProperties]
  renderers [{rendererClass=&NiagaraSpriteRendererProperties}]
 }
 SpeedLine {
  emitterUpdateScript [
   &EmitterState()
   &SpawnRate(
    Spawn Probability=$System.Timeline.Time
    SpawnRate=&FN_FrontEndFloatModifier(Gameplay=90, Front End=50)
    Use Spawn Probability=true
   ) {module=SpawnRate001}
  ]
  particleSpawnScript [
   &InitializeParticle(
    Color=&MakeLinearColorFromVectorAndFloat(
     "Vector (RGB)"=<1 1 1>
     "Float (Alpha)"=&Multiply_Float(
      A=&UniformRangedFloat(Minimum=0.1, Maximum=0.18)
      B=&MakeCustomFloatFromBool(Bool=$User.bIsFrontEndPreview, True Float=0.5, False Float=0.75)
     )
    )
    Lifetime Max=0.16
    Lifetime Min=0.1
    Lifetime Mode=::NewEnumerator1
    Sprite Size Max=<3 10>
    Sprite Size Min=<1 2>
    Sprite Size Mode=::NewEnumerator2
    UsePositionOffset=false
    Write Material Random=true
   )
   &TorusLocation(
    Align Torus With Axis=true
    Handle Radius=0
    Large Radius:NiagaraFloat=&RandomRangeFloat(Minimum=150, Maximum=200)
    MaskSpawnGroup=false
    Module.Override Randomness=false
    Offset=<0 -180 0>
    Torus Axis=<0 1 0>
    Torus Origin:NiagaraPosition=&Add_Vector(
     A=Linked() {linkedVariable {name=Particles.Position, type=NiagaraPosition}}
     B=&MakeVector(X=0, Y=&MakeCustomFloatFromBool(Bool=$User.bIsFrontEndPreview, True Float=400, False Float=0), Z=0)
    )
   )
   &AddVelocityFromPoint(
    Module.Use Falloff Distance=false
    Origin Offset=<0 -600 0>
    Velocity Strength=&Multiply_Float(
     A=&FN_FrontEndFloatModifier(Gameplay=600, Front End=0)
     B=&MakeCustomFloatFromBool(Bool=$User.bIsFrontEndPreview, True Float=-1, False Float=1)
    )
   )
   &AddVelocity(Velocity=&Multiply_Vector(
     A=&RandomRangeVector(Minimum=<0 300 0>, Maximum=<0 700 0>)
     B=&MakeVector(X=1, Y=&MakeCustomFloatFromBool(Bool=$User.bIsFrontEndPreview, True Float=-1, False Float=1), Z=1)
    ))
   &OffsetPosition(Position Offset=&FN_FrontEndVector3Modifier(Gameplay=<0 0 0>, Front End=<0 0 175>))
  ]
  particleUpdateScript [
   &SolveForcesAndVelocity(Clamp Velocity=false, Limit Acceleration=false)
   &ParticleState()
   &ScaleSpriteSize(
    Non-Uniform Curve Sprite Scale=DataInterface() {propertyValues {XCurve {keys [<at 0 0.5 RCIM_Cubic>,<at 0.5 1 RCIM_Cubic>]}, YCurve {keys [<at 0 0 RCIM_Cubic>,<at 0.5 1 RCIM_Cubic>]}}}
    Scale Sprite Size Mode=::NewEnumerator3
   )
   &ScaleSpriteSizeBySpeed(Max Scale Factor=<1 50>, Min Scale Factor=<0.01 0.05>, Velocity Threshold=1)
   &ScaleColor(
    Scale Alpha=&FloatFromCurve(
     FloatCurve=DataInterface() {propertyValues {Curve {keys [
         <at 0 0 RCIM_Cubic>
         <at 0.3759 1.0207>
         {arriveTangent=-2.221, interpMode=RCIM_Cubic, leaveTangent=-2.221, tangentMode=RCTM_User, time=1, value=0}
        ]}}}
     Scale Curve=&MakeCustomFloatFromBool(
      Bool=$User.bIsFrontEndPreview
      True Float=&MakeCustomFloatFromBool(Bool=$User.bIsFullyDeployed, True Float=1, False Float=0)
      False Float=&Lerp_Float(A=0, B=1, Alpha=$User.ForwardDot)
     )
    )
    Scale RGB=&VectorFromFloat(Value=1)
    ScaleA=true
    ScaleRGB=true
   )
  ]
  rendererClasses [&NiagaraSpriteRendererProperties]
  renderers [{rendererClass=&NiagaraSpriteRendererProperties}]
 }
 Trail_Brown {
  emitterSpawnScript [&FN_RibbonIDOffsetUtility()]
  emitterUpdateScript [
   &EmitterState()
   &FN_SpawnAnimTrail(Number of Anim Trails=2)
  ]
  particleSpawnScript [
   &FN_InitializeAnimTrailBoneIndecies(First Bone or Socket Index=4, Second Bone or Socket Index=5)
   &FN_InitializeAnimTrailBoneIndecies(AnimTrail Selection=1, First Bone or Socket Index=6, Second Bone or Socket Index=7) {module=FN_InitializeAnimTrailBoneIndecies001}
   &FN_InitializeAnimTrail(
    Color=&SelectLinearColorFromArray(
     Array Sampling Mode=::NewEnumerator1
     Direct Array Index=$System.intVariant
     Color Selection Array=DataInterface() {propertyValues {ColorData [<rgba 1 1 1 1>, <rgba 1 1 1 1>, <rgba 0.0409 0.0452 0.0612 1>]}}
     Random Seed=?
    )
    Lifetime=0.16
   )
   &AddVelocity(Velocity=&FN_FrontEndVector3Modifier(Gameplay=<0 2500 0>, Front End=<0 -2500 0>))
  ]
  particleUpdateScript [
   &InheritSourceMovement(
    Applied Position Scale=&Multiply_VectorByFloat(
     Vector=<1 1 1>
     Float=&FloatFromCurve(FloatCurve=DataInterface() {propertyValues {Curve {keys [
          {interpMode=RCIM_Cubic, tangentMode=RCTM_User, time=0.66, value=1}
          {arriveTangent=-1.9042, interpMode=RCIM_Cubic, leaveTangent=-1.9042, tangentMode=RCTM_User, time=1, value=0.5}
         ]}}})
    )
    ApplyForce=false
    ApplyPosition=true
    ApplyPositionFalloff=false
   )
   &SolveForcesAndVelocity(Clamp Velocity=false, Limit Acceleration=false)
   &FN_UpdateAnimTrail()
   &InterpolateOverTime(Initial Float Value:NiagaraFloat=$System.Timeline.Time, Target Float Value=1, Use Set Interpolation Rate=true)
   &ScaleColor(
    Color Scale:LinearColor=&FNTimeOfDayColor(
     Night Color=&SelectLinearColorFromArray(
      Array Sampling Mode=::NewEnumerator1
      Direct Array Index=$System.intVariant
      Color Selection Array=DataInterface() {propertyValues {ColorData [<rgba 0.1022 0.0895 0.113 1.3>, <rgba 0.2503 0.2696 0.1493 0.9043>, <rgba 0.572 0.6084 0.65 0.95>]}}
      Random Seed=?
     )
     Morning Color=&SelectLinearColorFromArray(
      Array Sampling Mode=::NewEnumerator1
      Direct Array Index=$System.intVariant
      Color Selection Array=DataInterface() {propertyValues {ColorData [<rgba 0.25 0.0625 0.1114 0.8>, <rgba 0.3763 0.27 0.1928 0.5391>, <rgba 1 0.8319 0.7136 0.8>]}}
      Random Seed=?
     )
     Day Color=&SelectLinearColorFromArray(
      Array Sampling Mode=::NewEnumerator1
      Direct Array Index=$System.intVariant
      Color Selection Array=DataInterface() {propertyValues {ColorData [<rgba 0.4271 0.112 0.0326 1>, <rgba 0.7011 0.5048 0.2804 0.45>, <rgba 1 1 1 1>]}}
      Random Seed=?
     )
     Evening Color=&SelectLinearColorFromArray(
      Array Sampling Mode=::NewEnumerator1
      Direct Array Index=$System.intVariant
      Color Selection Array=DataInterface() {propertyValues {ColorData [<rgba 0.25 0.0625 0.1114 0.8>, <rgba 0.3763 0.27 0.1928 0.5391>, <rgba 1 0.8319 0.7136 0.8>]}}
      Random Seed=?
     )
    )
    Scale Mode=3
   )
   &ScaleColor(
    Scale Alpha=&FN_FrontEndFloatModifier(Gameplay=$System.Timeline.Time, Front End=$System.Timeline002.Time)
    ScaleA=true
    ScaleRGB=true
   ) {module=ScaleColor003}
   /SetVariables(
    Particles.TipColor:LinearColor=&MultiplyLinearColor(Linear Color=$Particles.Color, Scale Factor=&MakeVector4(X=1, Y=1, Z=1, W=0.4))
    Particles.WingColor:LinearColor=&MultiplyLinearColor(
     Linear Color=$Particles.Color
     Scale Factor=&MakeVector4(
      X=1
      Y=1
      Z=1
      W=&FloatFromCurve(FloatCurve=DataInterface() {propertyValues {Curve {keys [<at 0 0>,<at 0.05 1>]}}}, Scale Curve=0.4)
     )
    )
   ) {isSetParameters=true}
  ]
  rendererClasses [&NiagaraRibbonRendererProperties]
  renderers [
   {rendererClass=&NiagaraRibbonRendererProperties}
   {rendererClass=&NiagaraRibbonRendererProperties, rendererIndex=1}
  ]
 }
 Trail_Green {
  emitterSpawnScript [&FN_RibbonIDOffsetUtility()]
  emitterUpdateScript [
   &EmitterState()
   &FN_SpawnAnimTrail(Number of Anim Trails=2)
  ]
  particleSpawnScript [
   &FN_InitializeAnimTrailBoneIndecies()
   &FN_InitializeAnimTrailBoneIndecies(AnimTrail Selection=1, First Bone or Socket Index=2, Second Bone or Socket Index=3) {module=FN_InitializeAnimTrailBoneIndecies001}
   &FN_InitializeAnimTrail(
    Color=&SelectLinearColorFromArray(
     Array Sampling Mode=::NewEnumerator1
     Direct Array Index=$System.intVariant
     Color Selection Array=DataInterface() {propertyValues {ColorData [<rgba 0 0.25 0.0208 1>, <rgba 1 1 1 1>, <rgba 0.0409 0.0452 0.0612 1>]}}
     Random Seed=?
    )
    Lifetime=0.16
   )
   &AddVelocity(Velocity=&FN_FrontEndVector3Modifier(Gameplay=<0 2500 0>, Front End=<0 -2500 0>))
  ]
  particleUpdateScript [
   &InheritSourceMovement(
    Applied Position Scale=&Multiply_VectorByFloat(
     Vector=<1 1 1>
     Float=&FloatFromCurve(FloatCurve=DataInterface() {propertyValues {Curve {keys [
          {interpMode=RCIM_Cubic, tangentMode=RCTM_User, time=0.66, value=1}
          {arriveTangent=-1.9042, interpMode=RCIM_Cubic, leaveTangent=-1.9042, tangentMode=RCTM_User, time=1, value=0.5}
         ]}}})
    )
    ApplyForce=false
    ApplyPosition=true
    ApplyPositionFalloff=false
   )
   &SolveForcesAndVelocity(Clamp Velocity=false, Limit Acceleration=false)
   &FN_UpdateAnimTrail()
   &InterpolateOverTime(Initial Float Value:NiagaraFloat=$System.Timeline.Time, Target Float Value=1, Use Set Interpolation Rate=true)
   &ScaleColor(
    Color Scale:LinearColor=&FNTimeOfDayColor(
     Night Color=&SelectLinearColorFromArray(
      Array Sampling Mode=::NewEnumerator1
      Direct Array Index=$System.intVariant
      Color Selection Array=DataInterface() {propertyValues {ColorData [<rgba 0.572 0.6084 0.65 0.95>, <rgba 0.2503 0.2696 0.1493 0.9043>, <rgba 0.572 0.6084 0.65 0.95>]}}
      Random Seed=?
     )
     Morning Color=&SelectLinearColorFromArray(
      Array Sampling Mode=::NewEnumerator1
      Direct Array Index=$System.intVariant
      Color Selection Array=DataInterface() {propertyValues {ColorData [<rgba 1 0.8319 0.7136 0.8>, <rgba 0.3763 0.27 0.1928 0.5391>, <rgba 1 0.8319 0.7136 0.8>]}}
      Random Seed=?
     )
     Day Color=&SelectLinearColorFromArray(
      Array Sampling Mode=::NewEnumerator1
      Direct Array Index=$System.intVariant
      Color Selection Array=DataInterface() {propertyValues {ColorData [<rgba 1 1 1 1>, <rgba 0.7011 0.5048 0.2804 0.45>, <rgba 1 1 1 1>]}}
      Random Seed=?
     )
     Evening Color=&SelectLinearColorFromArray(
      Array Sampling Mode=::NewEnumerator1
      Direct Array Index=$System.intVariant
      Color Selection Array=DataInterface() {propertyValues {ColorData [<rgba 1 0.8319 0.7136 0.8>, <rgba 0.3763 0.27 0.1928 0.5391>, <rgba 1 0.8319 0.7136 0.8>]}}
      Random Seed=?
     )
    )
    Scale Mode=3
   )
   &ScaleColor(
    Scale Alpha=&FN_FrontEndFloatModifier(Gameplay=$System.Timeline.Time, Front End=$System.Timeline002.Time)
    ScaleA=true
    ScaleRGB=true
   ) {module=ScaleColor003}
   /SetVariables(
    Particles.TipColor:LinearColor=&MultiplyLinearColor(Linear Color=$Particles.Color, Scale Factor=&MakeVector4(X=1, Y=1, Z=1, W=0.6))
    Particles.WingColor:LinearColor=&MultiplyLinearColor(
     Linear Color=$Particles.Color
     Scale Factor=&MakeVector4(
      X=1
      Y=1
      Z=1
      W=&FloatFromCurve(FloatCurve=DataInterface() {propertyValues {Curve {keys [<at 0 0>,<at 0.05 1>]}}})
     )
    )
   ) {isSetParameters=true}
  ]
  rendererClasses [&NiagaraRibbonRendererProperties]
  renderers [
   {rendererClass=&NiagaraRibbonRendererProperties}
   {rendererClass=&NiagaraRibbonRendererProperties, rendererIndex=1}
  ]
 }
}
summary {
 systemName=NS_Glider_FunkyFeather_Trail
 emitters [
  {emitterName=Notes, rendererClasses [&NiagaraMeshRendererProperties]}
  {emitterName=Trail_Green, rendererClasses [&NiagaraRibbonRendererProperties]}
  {emitterName=Trail_Brown, rendererClasses [&NiagaraRibbonRendererProperties]}
  {emitterName=SpeedLine, rendererClasses [&NiagaraSpriteRendererProperties]}
  {emitterName=Sparkle003, rendererClasses [&NiagaraSpriteRendererProperties]}
 ]
 userVariables [
  User.bIsFrontEnd {type=NiagaraBool}
  User.bIsFrontEndPreview {type=NiagaraBool}
  User.bIsFullyDeployed {type=NiagaraBool}
  User.ForwardDot {type=NiagaraFloat, defaultValue=0, description=""}
  User.ParaGlideLeanAlpha {type=NiagaraFloat, defaultValue=1, description=""}
  User.RibbonIDOffset {type=NiagaraInt32}
  User.RightDot {type=NiagaraFloat, defaultValue=0, description=""}
  User.Variant {type=NiagaraFloat, defaultValue=0, description=""}
 ]
}
